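; Blueprint function library for PCG elements, providing FilterDataByTags to pick tagged data entries that carry any of the requested tags.
; It loops over each requested tag, each input entry and that entry's tags, comparing strings with a case-sensitive exact match.
; Matches are appended to a local output array, so results are grouped by requested tag and an entry matching several tags appears more than once.
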(fn FilterDataByTags
  (params (InputData :type Array%20of%20PCGTagged%20Data%20Structures) (Tags :type Set%20of%20Strings))
  (bind _result (Utilities|Set|ToArray Tags))
  (for _array_element _result
    (for _array_element_2 InputData
      (bind _result_2 (Utilities|Set|ToArray (.Tags (PCG|TaggedData|BreakTaggedData _array_element_2))))
      (for _array_element_3 _result_2
        (if (Utilities|String|EqualExactly(String) _array_element_3 _array_element)
          (Utilities|Array|Add (LocalVariables|GetOutputData) _array_element_2)))))
  (return (LocalVariables|GetOutputData)))
