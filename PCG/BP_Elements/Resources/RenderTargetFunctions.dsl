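; Render target function library for PCG elements that captures top-down views of the world and processes them into masks.
; RT_RenderScene and RT_RenderWorldOrtho take one-shot orthographic captures, optionally limited to terrain or tagged actors, then clear capture references to avoid OFPA issues.
; Image operations (blur, noise, slope, color mask, arithmetic) mostly use RenderTargetProcessor, drawing a parameterized dynamic material into a new target.
; RT_DebugResult and RT_Visualize display results through a pass-through copy or post-process material slots.
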(fn CreateAndSetupMaterialInstance
  (params (Material :type Material%20Interface%20Object%20Reference) (RenderTarget2D_Map :type Map%20of%20Names%20to%20Texture%20Render%20Target%202D%20Object%20References) (ScalarParametersMap :type Map%20of%20Names%20to%20Float%20%28double-precision%29s) (VectorParametersMap :type Map%20of%20Names%20to%20Vectors))
  (bind _returnvalue (Rendering|Material|CreateDynamicMaterialInstance 0 Material "None" "None" :!class /Script/Engine.KismetMaterialLibrary))
  ;[1 1 1] Set Texture Params
  (bind _keys (Utilities|Map|Keys RenderTarget2D_Map))
  (for _array_element _keys
    (Rendering|Material|SetTextureParameterValue _returnvalue _array_element (Utilities|Map|Find RenderTarget2D_Map _array_element)))
  ;[1 1 1] Set Scalar Params
  (bind _keys_2 (Utilities|Map|Keys ScalarParametersMap))
  (for _array_element_2 _keys_2
    (Rendering|Material|SetScalarParameterValue _returnvalue _array_element_2 (Utilities|Map|Find ScalarParametersMap _array_element_2) :!class /Script/Engine.MaterialInstanceDynamic))
  ;[1 1 1] Set Vector Params
  (bind _keys_3 (Utilities|Map|Keys VectorParametersMap))
  (for _array_element_3 _keys_3
    (Rendering|Material|SetVectorParameterValue _returnvalue _array_element_3 (Math|Conversions|ToLinearColor(Vector) (Utilities|Map|Find VectorParametersMap _array_element_3)) :!class /Script/Engine.MaterialInstanceDynamic)
    (return _returnvalue)))

(fn DrawMaterialToNewRenderTarget
  (params (Material :type Material%20Interface%20Object%20Reference) (Width :type Integer) (Height :type Integer) (Format :type ETextureRenderTargetFormat%20Enum :default "RTF_R8"))
  (bind _returnvalue (Rendering|CreateRenderTarget2D 0 Width Height Format))
  (Variables|/Script/Engine.TextureRenderTarget2D|SetAddressX "TA_Clamp" _returnvalue)
  (Variables|/Script/Engine.TextureRenderTarget2D|SetAddressY "TA_Clamp" _returnvalue)
  (Rendering|DrawMaterialtoRenderTarget 0 _returnvalue Material)
  (return _returnvalue))

(fn RenderTargetProcessor
  (params (Material :type Material%20Interface%20Soft%20Object%20Reference) (RenderTarget2D_Map :type Map%20of%20Names%20to%20Texture%20Render%20Target%202D%20Object%20References) (ScalarParametersMap :type Map%20of%20Names%20to%20Float%20%28double-precision%29s) (VectorParametersMap :type Map%20of%20Names%20to%20Vectors) (Width :type Integer) (Height :type Integer) (Format :type ETextureRenderTargetFormat%20Enum :default "RTF_R8"))
  (bind _returnvalue (Utilities|LoadAssetBlocking Material))
  (bind _asmaterial_interface (Utilities|Casting|CastToMaterialInterface _returnvalue))
  (bind _materialinstance (CallFunction|CreateandSetupMaterialInstance _asmaterial_interface RenderTarget2D_Map ScalarParametersMap VectorParametersMap))
  (bind _rendertarget2d (CallFunction|DrawMaterialtoNewRenderTarget _materialinstance Width Height Format))
  (return _rendertarget2d))

(fn RT_AdjustLevels
  (params (RenderTargetIn :type Texture%20Render%20Target%202D%20Object%20Reference) (ScalarParametersMap :type Map%20of%20Names%20to%20Float%20%28double-precision%29s) (VectorParametersMap :type Map%20of%20Names%20to%20Vectors) (RT_Resolution :type Integer :default 1024))
  (bind _rendertarget2d (CallFunction|RenderTargetProcessor "/PCG/BP_Elements/Resources/MaterialsAndRTs/M_RT_AdjustLevels.M_RT_AdjustLevels" (Utilities|Map|MakeMap "SourceRenderTarget" RenderTargetIn) ScalarParametersMap VectorParametersMap RT_Resolution RT_Resolution "RTF_RGBA16f"))
  (return _rendertarget2d))

(fn RT_Blur
  (params (RenderTargetIn :type Texture%20Render%20Target%202D%20Object%20Reference) (ScalarParametersMap :type Map%20of%20Names%20to%20Float%20%28double-precision%29s) (VectorParametersMap :type Map%20of%20Names%20to%20Vectors) (RT_SizeInWorldUnits :type Float%20%28double-precision%29 :default 1.0) (RT_Resolution :type Integer :default 1024))
  (bind _keys (Utilities|Map|Keys ScalarParametersMap))
  (for _array_element _keys
    (Utilities|Map|Add (LocalVariables|GetLocalScalarParams) _array_element (Utilities|Map|Find ScalarParametersMap _array_element)))
  (Utilities|Map|Add (LocalVariables|GetLocalScalarParams) (LocalVariables|GetBlurDistance) (Math|Float|SafeDivide (Utilities|Map|Find ScalarParametersMap (LocalVariables|GetBlurDistance)) RT_SizeInWorldUnits))
  (bind _rendertarget2d (CallFunction|RenderTargetProcessor "/PCG/BP_Elements/Resources/MaterialsAndRTs/M_RT_Blur.M_RT_Blur" (Utilities|Map|MakeMap "SourceRenderTarget" RenderTargetIn) (LocalVariables|GetLocalScalarParams) VectorParametersMap RT_Resolution RT_Resolution "RTF_RGBA16f"))
  (return _rendertarget2d))

(fn RT_DebugResult
  (params (RenderTargetIn :type Texture%20Render%20Target%202D%20Object%20Reference) (RenderTargetOut :type Texture%20Render%20Target%202D%20Object%20Reference))
  (bind _materialinstance (CallFunction|CreateandSetupMaterialInstance "/PCG/BP_Elements/Resources/MaterialsAndRTs/M_RT_PassThrough.M_RT_PassThrough" (Utilities|Map|MakeMap "SourceRenderTargetA" RenderTargetIn) (Utilities|Map|MakeMap "None" 0.0) (Utilities|Map|MakeMap "None" "0, 0, 0")))
  (Rendering|DrawMaterialtoRenderTarget 0 RenderTargetOut _materialinstance)
  (return))

(fn RT_GenerateColorMask
  (params (RenderTargetIn :type Texture%20Render%20Target%202D%20Object%20Reference) (ScalarParametersMap :type Map%20of%20Names%20to%20Float%20%28double-precision%29s) (ColorParametersMap :type Map%20of%20Names%20to%20Linear%20Color%20Structures) (RT_Resolution :type Integer :default 1024))
  (bind _keys (Utilities|Map|Keys ColorParametersMap))
  (for _array_element _keys
    (Utilities|Map|Add (LocalVariables|GetvectorParams) _array_element (Math|Conversions|ToVector(LinearColor) (Utilities|Map|Find ColorParametersMap _array_element))))
  (bind _rendertarget2d (CallFunction|RenderTargetProcessor "/PCG/BP_Elements/Resources/MaterialsAndRTs/M_RT_ColorMask.M_RT_ColorMask" (Utilities|Map|MakeMap "SourceRenderTarget" RenderTargetIn) ScalarParametersMap (LocalVariables|GetvectorParams) RT_Resolution RT_Resolution "RTF_RGBA16f"))
  (return _rendertarget2d))

(fn RT_GenerateNormalVarianceMask
  (params (RenderTargetIn :type Texture%20Render%20Target%202D%20Object%20Reference) (ScalarParametersMap :type Map%20of%20Names%20to%20Float%20%28double-precision%29s) (VectorParametersMap :type Map%20of%20Names%20to%20Vectors) (RT_Resolution :type Integer :default 1024))
  (bind _rendertarget2d (CallFunction|RenderTargetProcessor "/PCG/BP_Elements/Resources/MaterialsAndRTs/M_RT_NormalVariance.M_RT_NormalVariance" (Utilities|Map|MakeMap "WorldNormals" RenderTargetIn) ScalarParametersMap VectorParametersMap RT_Resolution RT_Resolution "RTF_RGBA16f"))
  (return _rendertarget2d))

(fn RT_GenerateSlopeMask
  (params (RenderTargetIn :type Texture%20Render%20Target%202D%20Object%20Reference) (ScalarParametersMap :type Map%20of%20Names%20to%20Float%20%28double-precision%29s) (VectorParametersMap :type Map%20of%20Names%20to%20Vectors) (RT_Resolution :type Integer :default 1024))
  (bind _rendertarget2d (CallFunction|RenderTargetProcessor "/PCG/BP_Elements/Resources/MaterialsAndRTs/M_RT_SlopeMask.M_RT_SlopeMask" (Utilities|Map|MakeMap "WorldNormals" RenderTargetIn) ScalarParametersMap VectorParametersMap RT_Resolution RT_Resolution "RTF_RGBA16f"))
  (return _rendertarget2d))

(fn RT_Invert
  (params (RenderTargetIn :type Texture%20Render%20Target%202D%20Object%20Reference) (RT_Resolution :type Integer :default 1024))
  (bind _rendertarget2d (CallFunction|RenderTargetProcessor "/PCG/BP_Elements/Resources/MaterialsAndRTs/M_RT_Invert.M_RT_Invert" (Utilities|Map|MakeMap "SourceRenderTargetA" RenderTargetIn) (Utilities|Map|MakeMap "None" 0.0) (Utilities|Map|MakeMap "None" "0, 0, 0") RT_Resolution RT_Resolution "RTF_RGBA16f"))
  (return _rendertarget2d))

(fn RT_Multiply
  (params (RenderTargetInA :type Texture%20Render%20Target%202D%20Object%20Reference) (RenderTargetInB :type Texture%20Render%20Target%202D%20Object%20Reference) (RT_Resolution :type Integer :default 1024))
  (bind _rendertarget2d (CallFunction|RenderTargetProcessor "/PCG/BP_Elements/Resources/MaterialsAndRTs/M_RT_Multiply.M_RT_Multiply" (Utilities|Map|MakeMap "SourceRenderTargetA" RenderTargetInA "SourceRenderTargetB" RenderTargetInB) (Utilities|Map|MakeMap "None" 0.0) (Utilities|Map|MakeMap "None" "0, 0, 0") RT_Resolution RT_Resolution "RTF_RGBA16f"))
  (return _rendertarget2d))

(fn RT_MultiplyByFloat
  (params (RenderTargetIn :type Texture%20Render%20Target%202D%20Object%20Reference) (RT_Resolution :type Integer :default 1024) (Multiplier :type Float%20%28double-precision%29 :default 1.0))
  (bind _rendertarget2d (CallFunction|RenderTargetProcessor "/PCG/BP_Elements/Resources/MaterialsAndRTs/M_RT_MultiplyByFloat.M_RT_MultiplyByFloat" (Utilities|Map|MakeMap "SourceRenderTargetA" RenderTargetIn) (Utilities|Map|MakeMap "Multiplier" Multiplier) (Utilities|Map|MakeMap) RT_Resolution RT_Resolution "RTF_RGBA16f"))
  (return _rendertarget2d))

(fn RT_Noise
  (params (NoiseScale :type Float%20%28double-precision%29 :default 1024.0) (RT_SizeInWorldUnits :type Float%20%28double-precision%29 :default 1024.0) (RenderTargetIn :type Texture%20Render%20Target%202D%20Object%20Reference) (NoiseOffset :type Float%20%28double-precision%29) (NoiseMultiplier :type Float%20%28double-precision%29 :default 1.0) (RT_Resolution :type Integer :default 1024))
  (bind _localscalarparams (LocalVariables|GetLocalScalarParams))
  (Utilities|Map|Add _localscalarparams "NoiseScale" (Math|Float|SafeDivide NoiseScale RT_SizeInWorldUnits))
  (Utilities|Map|Add _localscalarparams "NoiseOffset" NoiseOffset)
  (Utilities|Map|Add _localscalarparams "NoiseMultiplier" NoiseMultiplier)
  (bind _rendertarget2d (CallFunction|RenderTargetProcessor "/PCG/BP_Elements/Resources/MaterialsAndRTs/M_RT_Noise.M_RT_Noise" (Utilities|Map|MakeMap "SourceRenderTarget" RenderTargetIn) (LocalVariables|GetLocalScalarParams) (Utilities|Map|MakeMap) RT_Resolution RT_Resolution "RTF_RGBA16f"))
  (return _rendertarget2d))

(fn RT_RenderScene
  (params (CaptureSource :type ESceneCaptureSource%20Enum :default "SCS_SceneColorHDR") (TargetActor :type Actor%20Object%20Reference) (RelativeLocation :type Vector) (Extents :type Vector) (RenderTerrainOnly :type Boolean) (TerrainActorClass :type Actor%20Class%20Reference) (ActorTag :type Array%20of%20Names) (TexelsPerMeter :type Float%20%28single-precision%29 :default 1.0) (RenderTargetFormat :type ETextureRenderTargetFormat%20Enum :default "RTF_RGBA32f") (PostProcessMaterial :type Material%20Interface%20Object%20Reference))
  (bind _actortag (Variables|GetActorTag))
  (bind _returnvalue (Utilities|Array|Length _actortag))
  (bind _texelspermeter (Variables|GetTexelsPerMeter))
  (bind _extents (Variables|GetExtents))
  (bind _y (.y _extents))
  (bind _x (.x _extents))
  (bind _returnvalue_2 (Math|Float|Max(Float) _x _y))
  (bind _returnvalue_3 (* _returnvalue_2 2.0))
  (bind _returnvalue_4 (* _returnvalue_3 0.01))
  (bind _returnvalue_5 (* _returnvalue_4 _texelspermeter))
  (bind _returnvalue_6 (Math|Float|Max(Float) _returnvalue_5 1.0))
  (bind _returnvalue_7 (Math|Float|Truncate _returnvalue_6))
  (bind _targetactor (Variables|GetTargetActor))
  (bind _returnvalue_8 (Actor|GetComponentbyClass _targetactor "/Script/Engine.SceneCaptureComponent2D"))
  (bind _returnvalue_9 (Utilities|IsValid _returnvalue_8 :!class /Script/Engine.KismetSystemLibrary))
  (bind _relativelocation (Variables|GetRelativeLocation))
  (bind _returnvalue_10 (Math|Transform|MakeTransform _relativelocation "-90.000000, 0,-90.000000" "0.100000,0.100000,0.100000"))
  (bind _renderterrainonly (Variables|GetRenderTerrainOnly))
  (bind _returnvalue_11 (> _returnvalue 0))
  (bind _returnvalue_12 (or _returnvalue_11 _renderterrainonly))
  (if _returnvalue_9
    (Transformation|SetRelativeTransform _returnvalue_8 _returnvalue_10 false false)
    (else
      (bind _returnvalue_13 (Game|AddComponentbyClass (Variables|GetTargetActor) 0 "/Script/Engine.SceneCaptureComponent2D" false _returnvalue_10))))
  (LocalVariables|SetCaptureComponent (select _returnvalue_9 _returnvalue_8 _returnvalue_13))
  (Variables|/Script/Engine.SceneCaptureComponent|SetbAlwaysPersistRenderingState true (LocalVariables|GetCaptureComponent))
  (Variables|/Script/Engine.SceneCaptureComponent|SetPrimitiveRenderMode (select _returnvalue_12 "PRM_UseShowOnlyList" "PRM_RenderScenePrimitives") (LocalVariables|GetCaptureComponent))
  (if _returnvalue_12
    (if (Variables|GetRenderTerrainOnly)
      (bind _outactors (Actor|GetAllActorsOfClass 0 (Variables|GetTerrainActorClass)))
      (LocalVariables|SetActorsToRender _outactors)
      (else
        (for _array_element (Variables|GetActorTag)
          (bind _outactors_2 (Actor|GetAllActorswithTag 0 _array_element))
          (Utilities|Array|AppendArray (LocalVariables|GetActorsToRender) _outactors_2))))
    (Variables|/Script/Engine.SceneCaptureComponent|SetShowOnlyActors (LocalVariables|GetActorsToRender) (LocalVariables|GetCaptureComponent))
    (else
      _))
  (bind _returnvalue_14 (Rendering|CreateRenderTarget2D 0 _returnvalue_7 _returnvalue_7 (Variables|GetRenderTargetFormat) "(R=0.000000,G=0.000000,B=0.000000,A=1.000000)" false true))
  (Variables|/Script/Engine.TextureRenderTarget2D|SetAddressX "TA_Clamp" _returnvalue_14)
  (Variables|/Script/Engine.TextureRenderTarget2D|SetAddressY "TA_Clamp" _returnvalue_14)
  (Variables|/Script/Engine.SceneCaptureComponent2D|SetTextureTarget _returnvalue_14 (LocalVariables|GetCaptureComponent))
  (Variables|/Script/Engine.SceneCaptureComponent2D|SetProjectionType "Orthographic" (LocalVariables|GetCaptureComponent))
  (Variables|/Script/Engine.SceneCaptureComponent2D|SetOrthoWidth _returnvalue_3 (LocalVariables|GetCaptureComponent))
  (Variables|/Script/Engine.SceneCaptureComponent|SetbCaptureEveryFrame false (LocalVariables|GetCaptureComponent))
  (Variables|/Script/Engine.SceneCaptureComponent|SetbCaptureOnMovement false (LocalVariables|GetCaptureComponent))
  (Variables|/Script/Engine.SceneCaptureComponent|SetCaptureSource (Variables|GetCaptureSource) (LocalVariables|GetCaptureComponent))
  (Variables|/Script/Engine.SceneCaptureComponent2D|SetPostProcessSettings (Utilities|Struct|MakePostProcessSettings (Utilities|Struct|MakeWeightedBlendables (Utilities|Array|MakeArray (Utilities|Struct|MakeWeightedBlendable 1.0 (Variables|GetPostProcessMaterial))))) (LocalVariables|GetCaptureComponent))
  (Utilities|IsValid (Variables|GetPostProcessMaterial)
    (:Is%20Valid
      (Variables|/Script/Engine.SceneCaptureComponent|SetShowFlagSettings (Utilities|Array|MakeArray (Utilities|Struct|MakeEngineShowFlagsSetting "Lighting")) (LocalVariables|GetCaptureComponent)))
    (:Is%20Not%20Valid
      _))
  (Rendering|SceneCapture|CaptureScene (LocalVariables|GetCaptureComponent) :!class /Script/Engine.SceneCaptureComponent2D)
  ;[1 1 1] clear refs to RT and landscape actor to avoid OFPA issues
  (Variables|/Script/Engine.SceneCaptureComponent2D|SetTextureTarget "" (LocalVariables|GetCaptureComponent))
  (Utilities|Array|Clear (LocalVariables|GetActorsToRender))
  (Variables|/Script/Engine.SceneCaptureComponent|SetShowOnlyActors (LocalVariables|GetActorsToRender) (LocalVariables|GetCaptureComponent))
  (return _returnvalue_14 _returnvalue_7))

(fn RT_RenderWorldOrtho
  (params (CaptureSource :type ESceneCaptureSource%20Enum :default "SCS_SceneColorHDR") (TargetActor :type Actor%20Object%20Reference) (Extents :type Vector) (RenderTerrainOnly :type Boolean :default true) (ActorTag :type Array%20of%20Names) (TexelsPerMeter :type Float%20%28single-precision%29 :default 1.0))
  (bind (_origin _boxextent) (Collision|GetActorBounds TargetActor false false))
  (bind _y (.y Extents))
  (bind _x (.x Extents))
  (bind _returnvalue (Math|Float|Max(Float) _x _y))
  (bind _returnvalue_2 (* _returnvalue 0.01))
  (bind _returnvalue_3 (* _returnvalue_2 TexelsPerMeter))
  (bind _returnvalue_4 (Math|Float|Max(Float) _returnvalue_3 1.0))
  (bind _returnvalue_5 (Math|Float|Truncate _returnvalue_4))
  (bind _returnvalue_6 (Actor|GetComponentbyClass TargetActor "/Script/Engine.SceneCaptureComponent2D"))
  (bind _returnvalue_7 (Utilities|IsValid _returnvalue_6 :!class /Script/Engine.KismetSystemLibrary))
  (bind _capturecomponent (LocalVariables|GetCaptureComponent))
  (if (not _returnvalue_7)
    (bind _returnvalue_8 (Game|AddComponentbyClass TargetActor 0 "/Script/Engine.SceneCaptureComponent2D" false (Math|Transform|MakeTransform (Math|Vector|MakeVector 0.0 0.0 (.z (/ _boxextent (Transformation|GetActorScale3D TargetActor)))) "-90.000000, 0,-90.000000"))))
  (LocalVariables|SetCaptureComponent (select _returnvalue_7 _returnvalue_6 _returnvalue_8))
  (Variables|/Script/Engine.SceneCaptureComponent|SetbAlwaysPersistRenderingState true _capturecomponent)
  (Transformation|SetRelativeScale3D _capturecomponent "0.100000,0.100000,0.100000")
  (Variables|/Script/Engine.SceneCaptureComponent|SetShowFlagSettings (Utilities|Array|MakeArray (Utilities|Struct|MakeEngineShowFlagsSetting "instancedgrass")) (LocalVariables|GetCaptureComponent))
  (Variables|/Script/Engine.SceneCaptureComponent|SetPrimitiveRenderMode (select (and (< (Utilities|Array|Length ActorTag) 1) (not RenderTerrainOnly)) "PRM_RenderScenePrimitives" "PRM_UseShowOnlyList") (LocalVariables|GetCaptureComponent))
  (if RenderTerrainOnly
    (bind _outactors (Actor|GetAllActorsOfClass 0 "/Script/Landscape.LandscapeProxy"))
    (LocalVariables|SetActorsToRender _outactors)
    (else
      (for _array_element ActorTag
        (bind _outactors_2 (Actor|GetAllActorswithTag 0 _array_element))
        (Utilities|Array|AppendArray (LocalVariables|GetActorsToRender) _outactors_2))))
  (Variables|/Script/Engine.SceneCaptureComponent|SetShowOnlyActors (LocalVariables|GetActorsToRender) (LocalVariables|GetCaptureComponent))
  (bind _returnvalue_9 (Rendering|CreateRenderTarget2D 0 _returnvalue_5 _returnvalue_5))
  (Variables|/Script/Engine.TextureRenderTarget2D|SetAddressX "TA_Clamp" _returnvalue_9)
  (Variables|/Script/Engine.TextureRenderTarget2D|SetAddressY "TA_Clamp" _returnvalue_9)
  (Variables|/Script/Engine.SceneCaptureComponent2D|SetTextureTarget _returnvalue_9 (LocalVariables|GetCaptureComponent))
  (Variables|/Script/Engine.SceneCaptureComponent2D|SetProjectionType "Orthographic" (LocalVariables|GetCaptureComponent))
  (Variables|/Script/Engine.SceneCaptureComponent2D|SetOrthoWidth (* _returnvalue 2.0) (LocalVariables|GetCaptureComponent))
  (Variables|/Script/Engine.SceneCaptureComponent|SetbCaptureEveryFrame false (LocalVariables|GetCaptureComponent))
  (Variables|/Script/Engine.SceneCaptureComponent|SetbCaptureOnMovement false (LocalVariables|GetCaptureComponent))
  (Variables|/Script/Engine.SceneCaptureComponent|SetCaptureSource CaptureSource (LocalVariables|GetCaptureComponent))
  (Rendering|SceneCapture|CaptureScene (LocalVariables|GetCaptureComponent) :!class /Script/Engine.SceneCaptureComponent2D)
  ;[1 1 1] clear refs to RT and landscape actor to avoid OFPA issues
  (Variables|/Script/Engine.SceneCaptureComponent2D|SetTextureTarget "" (LocalVariables|GetCaptureComponent))
  (Variables|/Script/Engine.SceneCaptureComponent|SetShowOnlyActors (Utilities|Array|MakeArray 0) (LocalVariables|GetCaptureComponent))
  (return _returnvalue_9 _returnvalue_5))

(fn RT_RtoRBG
  (params (RenderTargetIn :type Texture%20Render%20Target%202D%20Object%20Reference) (RT_Resolution :type Integer :default 1024))
  (bind _rendertarget2d (CallFunction|RenderTargetProcessor "/PCG/BP_Elements/Resources/MaterialsAndRTs/M_RT_RtoRBG.M_RT_RtoRBG" (Utilities|Map|MakeMap "SourceRenderTargetA" RenderTargetIn) (Utilities|Map|MakeMap) (Utilities|Map|MakeMap) RT_Resolution RT_Resolution "RTF_RGBA16f"))
  (return _rendertarget2d))

(fn RT_Saturate
  (params (RenderTargetIn :type Texture%20Render%20Target%202D%20Object%20Reference) (RT_Resolution :type Integer :default 1024))
  (bind _rendertarget2d (CallFunction|RenderTargetProcessor "/PCG/BP_Elements/Resources/MaterialsAndRTs/M_RT_Saturate.M_RT_Saturate" (Utilities|Map|MakeMap "SourceRenderTargetA" RenderTargetIn) (Utilities|Map|MakeMap "None" 0.0) (Utilities|Map|MakeMap "None" "0, 0, 0") RT_Resolution RT_Resolution "RTF_RGBA16f"))
  (return _rendertarget2d))

(fn RT_Subtract
  (params (RenderTargetInA :type Texture%20Render%20Target%202D%20Object%20Reference) (RenderTargetInB :type Texture%20Render%20Target%202D%20Object%20Reference) (RT_Resolution :type Integer :default 1024))
  (bind _rendertarget2d (CallFunction|RenderTargetProcessor "/PCG/BP_Elements/Resources/MaterialsAndRTs/M_RT_Subtract.M_RT_Subtract" (Utilities|Map|MakeMap "SourceRenderTargetA" RenderTargetInA "SourceRenderTargetB" RenderTargetInB) (Utilities|Map|MakeMap "None" 0.0) (Utilities|Map|MakeMap "None" "0, 0, 0") RT_Resolution RT_Resolution "RTF_RGBA16f"))
  (return _rendertarget2d))

(fn RT_Visualize
  (params (Target%20Actor :type Actor%20Object%20Reference) (Render%20Target :type Texture%20Render%20Target%202D%20Object%20Reference) (SlotIndex :type Integer))
  (bind _componenttag (LocalVariables|GetComponentTag))
  (bind _target_actor (Variables|GetTarget%20Actor))
  (bind _returnvalue (Actor|GetComponentsbyTag _target_actor "/Script/Engine.PostProcessComponent" _componenttag))
  (LocalVariables|SetTargetActor Target%20Actor)
  (Utilities|FlowControl|Sequence
    (:then_0
      (if (> (Utilities|Array|Length _returnvalue) 0)
        (bind _output_get (LocalVariables|SetPPComponent (Utilities|Array|Get(acopy) _returnvalue)))
        (bind _asmaterial_instance_dynamic (Utilities|Casting|CastToMaterialInstanceDynamic (.Object (Utilities|Struct|BreakWeightedBlendable (Utilities|Array|Get(aref) (Utilities|Struct|BreakWeightedBlendables (Utilities|Struct|BreakPostProcessSettings (Variables|/Script/Engine.PostProcessComponent|GetSettings _output_get)))))))
          (:then
            (LocalVariables|SetPPMaterialInstance _asmaterial_instance_dynamic))
          (:CastFailed
            _))
        (else
          (bind _returnvalue_2 (Game|AddComponentbyClass :self (Variables|GetTarget%20Actor) :Class "/Script/Engine.PostProcessComponent"))
          (Variables|/Script/Engine.ActorComponent|SetComponentTags (Utilities|Array|MakeArray (LocalVariables|GetComponentTag)) _returnvalue_2)
          (LocalVariables|SetPPComponent _returnvalue_2)))
      (bind _returnvalue_3 (Rendering|Material|CreateDynamicMaterialInstance 0 (LocalVariables|GetPPMaterial) "None" "Transient" :!class /Script/Engine.KismetMaterialLibrary))
      (LocalVariables|SetPPMaterialInstance _returnvalue_3)
      (Variables|/Script/Engine.PostProcessComponent|SetSettings (Utilities|Struct|MakePostProcessSettings (Utilities|Struct|MakeWeightedBlendables (Utilities|Array|MakeArray (Utilities|Struct|MakeWeightedBlendable 1.0 _returnvalue_3)))) (LocalVariables|GetPPComponent)))
    (:then_1
      (Variables|/Script/Engine.PostProcessComponent|SetbUnbound true (LocalVariables|GetPPComponent))
      (Rendering|Material|SetTextureParameterValue (LocalVariables|GetPPMaterialInstance) (Utilities|String|StringToName (Utilities|String|Append "Slot" (Utilities|String|ToString(Integer) SlotIndex))) Render%20Target))))
